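# Sound cue for the FirePetal day/night forest's time-of-day change, in the quest sound class.
# A quality-level node picks the quad wave for the first two quality levels and the stereo wave for the last two.
# It references the shared Athena ambient attenuation asset and also stores non-spatialized override settings with a custom long-range curve that holds volume until near the edge.
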SoundCue /FirePetalDayNightForest/Assets/Audio/FirePetalDayNight_TOD_Cue
root=QualityLevel_0
nodes [
 {id=QualityLevel_0, children [WavePlayer_1,WavePlayer_1,WavePlayer_2,WavePlayer_2]}
 {id=WavePlayer_1, properties {soundWaveAssetPtr=&/FirePetalDayNight_TODChange_Quad}}
 {id=WavePlayer_2, properties {soundWaveAssetPtr=&/FirePetalDayNight_TODChange_Stereo}}
]
properties {
 attenuationSettings=&Fort_Ambient_Athena_Att
 maxDistance=2097152
 soundClassObject=&Fort_Quests
 attenuationOverrides {
  bSpatialize=false
  distanceAlgorithm=Custom
  falloffDistance=33000
  attenuationShapeExtents {x=100}
  customAttenuationCurve {editorCurveData {keys [
     {arriveTangent=-1.5973, interpMode=RCIM_Cubic, time=0, value=1}
     <at 0.85 1>
     {arriveTangent=-3.0808, interpMode=RCIM_Cubic, leaveTangent=-3.0808, tangentMode=RCTM_User, time=1, value=0}
    ]}}
 }
}
