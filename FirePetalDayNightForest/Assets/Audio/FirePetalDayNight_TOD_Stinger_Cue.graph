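# Sound Cue for the time-of-day change stinger in FirePetal Day/Night Forest.
# A Mixer layers the TODChange Stinger wave with a TODChange BushRanger wave at equal volume, so both play together.
# The cue routes to the Fort_Quests sound class.
# It references a musical one-shot, no-reverb attenuation asset, but Override Attenuation is on, so the cue's own attenuation settings take precedence.
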SoundCue /FirePetalDayNightForest/Assets/Audio/FirePetalDayNight_TOD_Stinger_Cue
root=Mixer_0
nodes [
 {id=Mixer_0, children [WavePlayer_1,WavePlayer_0], properties {inputVolume [1,1]}}
 {id=WavePlayer_0, properties {soundWaveAssetPtr=&/FirePetalDayNight_TODChange_BushRanger}}
 {id=WavePlayer_1, properties {soundWaveAssetPtr=&/FirePetalDayNight_TODChange_Stinger}}
]
properties {
 attenuationSettings=/Game/Sounds/Attenuation/Fort_Ambient_Athena_Musical_Oneshot_NoVerb_Att
 bOverrideAttenuation=true
 maxDistance=4000
 soundClassObject=&Fort_Quests
 volumeMultiplier=1
}
