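# Sound cue for the looping ambient sound of the FirePetal vault door in the DayNightForest content, classed as an Athena ambient prop sound.
# A single looping wave player plays the vault door ambient loop wave directly as the root, with no mixing or modulation.
# A dedicated attenuation asset shapes its spatial falloff within a limited audible range.
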SoundCue /FirePetalDayNightForest/Assets/Audio/SC_FirePetalVaultDoor_Ambient_Loop
root=WavePlayer_0
nodes [{id=WavePlayer_0, properties {bLooping=true, soundWaveAssetPtr=&/FirePetalVaultDoor_Amb_Loop}}]
properties {
 attenuationSettings=&/ATT_FirePetalVaultDoor_Ambient_Loop
 maxDistance=1100
 soundClassObject=&Fort_Ambient_Props_Athena
 volumeMultiplier=1
}
