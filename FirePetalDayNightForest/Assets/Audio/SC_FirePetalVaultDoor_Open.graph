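# Sound cue for the opening sound of the Fire Petal vault door in the FirePetalDayNightForest content.
# A single wave player feeds the FirePetalVaultDoor_Open wave straight to the output, with no randomization or mixing.
# The cue uses a dedicated attenuation asset, routes to the Athena ambient props sound class, and boosts volume above default.
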SoundCue /FirePetalDayNightForest/Assets/Audio/SC_FirePetalVaultDoor_Open
root=WavePlayer_0
nodes [{id=WavePlayer_0, properties {soundWaveAssetPtr=&/FirePetalVaultDoor_Open}}]
properties {
 attenuationSettings=&/ATT_FirePetalVaultDoor_Open
 maxDistance=16000
 soundClassObject=&Fort_Ambient_Props_Athena
 volumeMultiplier=3
}
